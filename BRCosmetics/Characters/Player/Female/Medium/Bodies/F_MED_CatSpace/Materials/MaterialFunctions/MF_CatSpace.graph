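# Material function for the CatSpace character body cosmetic that adds thin film and animated gradient emissive to incoming material attributes.
# Pre-skinned Y position, passed through a vertex interpolator, blends back and front intensities for the nested thin-film function.
# The emissive texture's red channel is remapped into a gradient lookup coordinate and offset by time to scroll colors.
# That color is masked, divided by eye adaptation to offset exposure, and written to EmissiveColor.
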MaterialFunction /BRCosmetics/Characters/Player/Female/Medium/Bodies/F_MED_CatSpace/Materials/MaterialFunctions/MF_CatSpace
graph {nodes [
  n1 = PreSkinnedPosition()
  n2 = ShaderStageSwitch(PixelShader=VertexInterpolator(VS=@n1).PS, VertexShader=@n1)
  n3 = Constant() {r=0.03}
  n4 = Add(A=Multiply(A=ComponentMask(None=@n2) {g=true}, B=@n3) {constB=0.1}, B=Multiply(A=@n3, B=Constant() {r=34})) {constB=-100}
  n5 = MaterialFunctionCall!(
   CustomThinFilmIntensity=LinearInterpolate(
    A=ScalarParameter() {defaultValue=0.5, group=06_Thin Film, parameterName=ThinFilmIntensity_Back, sortPriority=2}
    B=ScalarParameter() {defaultValue=0.5, group=06_Thin Film, parameterName=ThinFilmIntensity_Front, sortPriority=2}
    Alpha=Saturate(None=@n4)
   )
   Material Attributes=FunctionInput() {inputName=Material Attributes, inputType=FunctionInput_MaterialAttributes}
   Fresnel=FunctionInput() {inputName=Fresnel, inputType=FunctionInput_Scalar}
   Alpha=FunctionInput() {inputName=TechArtMasks}
  ) {materialFunction=&/MF_ThinFilm_TechArt_CatSpace_Updated}
  n6 = TextureSampleParameter2D() {
   group=06_Emissive
   parameterName=Emissive
   texture=/BRCosmetics/Characters/Player/Female/Medium/Bodies/F_MED_CatSpace/Textures/T_CatSpace_Body_E
  }
  n7 = Saturate(None=InverseLinearInterpolate(A=Constant3Vector() {constant=<rgba 0.0103 0 0.0953 1>}.R, B=Constant3Vector() {constant=<rgba 1 0 0.147 1>}.R, Value=@n6.R))
  n8 = Multiply(A=Time(), B=ScalarParameter() {defaultValue=0.2, group=CatSpace, parameterName=EmissiveScrollingSpeed}) {constB=0.2}
  n9 = Multiply(
   A=TextureSample(UVs=Add(A=@n7, B=@n8)) {texture=/BRCosmetics/Characters/Player/Female/Medium/Bodies/F_MED_CatSpace/Textures/CatSpace_Gradient}.RGB
   B=FunctionInput() {inputName=EmissiveMask, inputType=FunctionInput_Scalar, previewValue=<1 0 0 1>}
  )
  n10 = EyeAdaptation()
  FunctionOutput(None=SetMaterialAttributes!(MaterialAttributes=@n5.Result, EmissiveColor=Divide(A=@n9, B=@n10.EyeAdaptation)) {attributeSetTypes [#1]})
 ]}
