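# Parent material for the CatSpace female medium player body, using material attributes and a shading model chosen by expression.
# The shared 2023 character PreFX function feeds the outfit-specific MF_CatSpace function, whose result passes into the shared PostFX function along with other PreFX outputs.
# MF_CatSpace also takes the red channel of the body emissive mask texture.
# Usage flags cover skeletal meshes, morph targets and mesh deformers.
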Material /BRCosmetics/Characters/Player/Female/Medium/Bodies/F_MED_CatSpace/Materials/UpdatedBodyMaterial/M_CatSpace_Parent_Updated
graph {
 properties {
  bUseMaterialAttributes=true
  bUsedWithMeshDeformer=true
  bUsedWithMorphTargets=true
  bUsedWithSkeletalMesh=true
  pixelDepthOffsetMode=PDOM_Legacy
  shadingModel=MSM_FromMaterialExpression
  usedShadingModels=Default Lit | Subsurface | Clear Coat | Subsurface Profile | Cloth
 }
 nodes [
  n1 = MaterialFunctionCall() {materialFunction=&MF_CharacterParent_2023_PreFX}
  n2 = TextureSample() {
   samplerType=SAMPLERTYPE_Masks
   texture=/BRCosmetics/Characters/Player/Female/Medium/Bodies/F_MED_CatSpace/Textures/T_CatSpace_Body_EmissiveMasks
  }
  MaterialFunctionCall(
   Material=MaterialFunctionCall!(TechArtMasks=@n1.Tech Art Mask, Material Attributes=@n1.Tech Art Mask, EmissiveMask=@n2.R, Fresnel=@n1.Tech Art Mask) {materialFunction=/BRCosmetics/Characters/Player/Female/Medium/Bodies/F_MED_CatSpace/Materials/MaterialFunctions/MF_CatSpace}.Result
   M=@n1.M
   Pre Skinned Local Position=@n1.M
   Distance Blend=@n1.M
   Fresnel=@n1.M
  ) {materialFunction=&MF_CharacterParent_2023_PostFX}
 ]
}
